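# Niagara effect for the swap moment of the Flourish emote, a short puff of smoke with streaks near the ground.
# Four emitters run: torus-spawned smoke fingers pushed outward, two expanding upward-facing ground rings, brief cone-launched streaks, and drifting cone-burst smoke skipped in front-end preview.
# Curves over particle life drive size, alpha and dynamic material parameters.
# Five other emitters are disabled, including all active skeletal-mesh sampling (embers, shield decals, spinning meshes).
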NiagaraSystem /BRCosmetics/Effects/Fort_Effects/Emotes/Flourish/NS_Emote_Flourish_Swap
dependencies {
 usedDataInterfaces [
  &NiagaraDataInterfaceCurve
  &NiagaraDataInterfaceCurlNoise
  &NiagaraDataInterfaceVector2DCurve
  &NiagaraDataInterfaceColorCurve
 ]
 usedDynamicInputs [&ReturnNormalizedExecIndex]
 usedModules [&SystemState]
}
emitters {
 BurstEmbers {
  bEnabled=false
  emitterUpdateScript [
   &EmitterState(Life Cycle Mode=::NewEnumerator1, Loop Behavior=::NewEnumerator1, Loop Delay=0.15, Loop Duration=0.4, UseLoopDelay=true)
   &SpawnBurst_Instantaneous(
    Spawn Count=&RandomRangeInt(Minimum=150, Maximum=180)
    Spawn Probability=&MakeCustomFloatFromBool(Bool=$User.bIsFrontEndPreview, True Float=0, False Float=1)
    Use Spawn Probability=true
   )
  ]
  particleSpawnScript [
   &FN_SetSpawnGroup(Number of Spawn Groups=17)
   &InitializeParticle(
    Color=<rgba 0.025 0.0474 0.1 1>
    Lifetime Max=0.5
    Lifetime Mode=::NewEnumerator1
    Sprite Size=&Multiply_Vector2D_ByFloat(Float=2, Vector2D=&RandomRangeVector2D(Minimum=<4 5>, Maximum=<4 5>))
    Sprite Size Mode=::NewEnumerator4
    UsePositionOffset=false
    Write Material Random=true
   ) {module=InitializeParticle001}
   &AddVelocity(Velocity=&RandomRangeVector(Minimum=<-20 -20 -20>, Maximum=<20 20 20>, Evaluation Type=::NewEnumerator1))
   &SkeletalMeshLocation(
    Bone / Socket Index=$Particles.SpawnGroup
    Bone Sampling Mode=::NewEnumerator4
    Inherit Velocity=false
    MaskSpawnGroup=false
    Skeletal Mesh=$User.SkeletalMesh
   )
   &ShapeLocation(MaskSpawnGroup=false, Override Seed=false, Sphere Radius=13, UseOwnerScale=true)
   &AddVelocity(Velocity=&Multiply_VectorByFloat(
     Vector=$Particles.ShapeLocation.ShapeVector
     Float=&RandomRangeFloat(Minimum=15, Maximum=25, Evaluation Type=::NewEnumerator1)
    )) {module=AddVelocity001}
  ]
  particleUpdateScript [
   &ParticleState()
   &ScaleSpriteSize(
    Scale Sprite Size Mode=::NewEnumerator0
    Uniform Scale Factor=&MakeFloatFromVector(
     Vector=&VectorFromFloat(Value=&Waveform(
       "Waveform [1]"=::NewEnumerator12
       "[1] Input Position"=&RandomRangeFloat(Minimum=0, Maximum=1000)
       "[1] Frequency"=25
       "[1] Amplitude Min/Max"=<0.5 1>
      ))
     Channel=::NewEnumerator0
    )
   ) {module=ScaleSpriteSize001}
   &ScaleSpriteSize(Scale Factor=&Vector2DFromFloat(Value=&FloatFromCurve(FloatCurve=DataInterface() {propertyValues {Curve {keys [
          {arriveTangent=11.6538, interpMode=RCIM_Cubic, time=0, value=0.0211}
          {arriveTangent=0.3289, interpMode=RCIM_Cubic, leaveTangent=0.3289, time=0.125, value=1.5}
          <at 1 0.35 RCIM_Cubic>
         ]}}})))
   &Drag(Drag=4.5, Use Linear Drag=true, Use Rotational Drag=false)
   &AccelerationForce(Acceleration=<0 0 50>)
   &SolveForcesAndVelocity(Clamp Velocity=false, Limit Acceleration=false)
   &ScaleSpriteSizeBySpeed(Max Scale Factor=<1 4>, Source Velocity:Vector3f=&Multiply_Vector(A=<0 0.01 1>, B=$Particles.Velocity))
   &Color(Color=<rgba 1.0997 2.3998 5 1>) {module=Color002}
   &ScaleColor(
    Color Value To Scale=$Particles.Color
    Scale Alpha=&FloatFromCurve(FloatCurve=DataInterface() {propertyValues {Curve {keys [<at 0 0 RCIM_Cubic>,<at 0.1 1 RCIM_Cubic>,<at 1 0>]}}})
    Scale RGB=&VectorFromFloat(Value=3)
    ScaleA=true
    ScaleRGB=true
   ) {module=ScaleColor001}
  ]
  rendererClasses [&NiagaraSpriteRendererProperties]
  renderers [
   {rendererClass=&NiagaraSpriteRendererProperties}
   {rendererClass=&NiagaraSpriteRendererProperties, rendererIndex=1}
  ]
 }
 CircularSmokeRate {
  bEnabled=false
  emitterUpdateScript [
   &EmitterState(Life Cycle Mode=::NewEnumerator1, Loop Behavior=::NewEnumerator1, Loop Delay=0.9083, Loop Duration=1, UseLoopDelay=false)
   &SpawnBurst_Instantaneous(Spawn Count=125, Use Spawn Probability=false, UseLoopCountLimit=false) {module=SpawnBurst_Instantaneous001}
   &SpawnBurst_Instantaneous(Spawn Count=40, Use Spawn Probability=false, UseLoopCountLimit=false) {enabled=false}
  ]
  particleSpawnScript [
   &InitializeParticle(
    Lifetime=&RandomRangeFloat(Minimum=0.6, Maximum=0.4)
    Position Offset=<0 0 1>
    Position Offset Coordinate Space=::Local
    Sprite Rotation Mode=::NewEnumerator2
    Sprite Size Mode=::NewEnumerator1
    Uniform Sprite Size Max=40
    Uniform Sprite Size Min=30
    UsePositionOffset=true
    Write Material Random=true
   )
   &SkeletalMeshLocation(MaskSpawnGroup=false, Orientation Sampling=::NewEnumerator1, OverrideRandomSeed=false, Skeletal Mesh=$User.Skeletal Mesh01) {enabled=false}
   &FN_SetSpawnGroup(Number of Spawn Groups=2) {enabled=false}
   &SystemLocation(MaskSpawnGroup=false, Offset=<-10 25 1>) {enabled=false}
   &TorusLocation(Align Torus With Axis=false, Large Radius=30, MaskSpawnGroup=false, Module.Override Randomness=false, Torus Mode=::NewEnumerator2)
   &AddVelocityFromPoint(Module.Use Falloff Distance=false, Velocity Strength=&RandomRangeFloat(Minimum=1500, Maximum=0))
   &AddVelocity(Velocity=<0 0 10>)
  ]
  particleUpdateScript [
   &CurlNoiseForce(Noise Strength=100, Pan Noise=false) {enabled=false}
   &GravityForce(Gravity=<0 0 50>) {enabled=false}
   &AccelerationForce(Acceleration=&RandomRangeVector(Minimum=<0 0 400>, Maximum=<0 0 600>))
   &Drag(Drag=&RandomRangeFloat(Minimum=12, Maximum=10), Use Linear Drag=true, Use Rotational Drag=false)
   &SolveForcesAndVelocity(Clamp Velocity=false, Limit Acceleration=false)
   &Color(
    Color=&MakeLinearColorFromVectorAndFloat(
     "Vector (RGB)"=&MakeVectorFromLinearColorRGB(Color=&FNTimeOfDayColor(
       Night Color=<rgba 0.5 0.7241 1 1>
       Morning Color=<rgba 0.5 0.7241 1 1>
       Day Color=<rgba 0.42 0.6799 1 1>
       Evening Color=<rgba 0.5 0.7241 1 1>
      ))
     "Float (Alpha)"=&FloatFromCurve(FloatCurve=DataInterface() {propertyValues {Curve {keys [
          <at 0 1>
          {arriveTangentWeight=0.2333, interpMode=RCIM_Cubic, leaveTangentWeight=0.1, tangentMode=RCTM_User, time=0.7, value=1}
          {arriveTangentWeight=0.1, interpMode=RCIM_Cubic, tangentMode=RCTM_User, time=1, value=0}
         ]}}})
    )
    Scale Alpha=&FloatFromCurve(FloatCurve=DataInterface() {propertyValues {Curve {keys [<at 0.8 1>,<at 1 0>]}}}, Scale Curve=0.5)
   )
   &ParticleState()
   &DynamicMaterialParameters(
    Index 0 Param 1=&FloatFromCurve(FloatCurve=DataInterface() {propertyValues {Curve {keys [<at 0 1>,<at 1 2>]}}}, Scale Curve=0.2)
    Index 0 Param 3=&FloatFromCurve(
     FloatCurve=DataInterface() {propertyValues {Curve {keys [
         {arriveTangent=-0.9531, interpMode=RCIM_Cubic, leaveTangent=-0.9531, tangentMode=RCTM_User, time=0, value=0.07}
         {arriveTangent=-0.3362, interpMode=RCIM_Cubic, leaveTangent=-0.3362, tangentMode=RCTM_User, time=0.1, value=0.01}
         {interpMode=RCIM_Cubic, leaveTangent=1.7757, time=0.15, value=0}
        ]}}}
     Scale Curve=0
    )
    Index 1 Param 1=&FloatFromCurve(FloatCurve=DataInterface() {propertyValues {Curve {keys [<at 0 0>,<at 1 1>]}}})
    Index 1 Param 2=&FloatFromCurve(FloatCurve=DataInterface() {propertyValues {Curve {keys [{arriveTangent=1.9694, interpMode=RCIM_Cubic, leaveTangent=1.9694, tangentMode=RCTM_User, time=0, value=0}, <at 1 1 RCIM_Cubic>]}}})
    Index 1 Param 3=&FloatFromCurve(FloatCurve=DataInterface() {propertyValues {Curve {keys [{arriveTangent=6.205, interpMode=RCIM_Cubic, leaveTangent=6.205, tangentMode=RCTM_User, time=0, value=0.2}, <at 0.25 1 RCIM_Cubic>]}}})
    Index 1 Param 4=&FloatFromCurve(FloatCurve=DataInterface() {propertyValues {Curve {keys [<at 0 1>,<at 1 0>]}}})
    Index 2 Param 1=&FloatFromCurve(FloatCurve=DataInterface() {propertyValues {Curve {keys [{arriveTangent=2.2804, interpMode=RCIM_Cubic, leaveTangent=2.2804, tangentMode=RCTM_User, time=0, value=0.1}, <at 1 1 RCIM_Cubic>]}}})
    Index 2 Param 2=&FloatFromCurve(FloatCurve=DataInterface() {propertyValues {Curve {keys [<at 0 0.2>,<at 1 0.8>]}}})
    Index 2 Param 3=&FloatFromCurve(FloatCurve=DataInterface() {propertyValues {Curve {keys [{interpMode=RCIM_Cubic, tangentMode=RCTM_User, time=0, value=1}, {interpMode=RCIM_Cubic, tangentMode=RCTM_User, time=1, value=2}]}}})
    Index 2 Param 4=&FloatFromCurve(FloatCurve=DataInterface() {propertyValues {Curve {keys [<at 0 1 RCIM_Cubic>,<at 1 0.75 RCIM_Cubic>]}}})
    Module.Param0WriteEnabled=true
    Module.Param0WriteEnabled1=true
    Module.Param0WriteEnabled2=true
    Module.Param0WriteEnabled3=true
    Module.Param1WriteEnabled=true
    Module.Param1WriteEnabled1=true
    Module.Param1WriteEnabled2=true
    Module.Param1WriteEnabled3=true
    Module.Param2WriteEnabled=true
    Module.Param2WriteEnabled1=true
    Module.Param2WriteEnabled2=true
    Module.Param2WriteEnabled3=true
    Module.Param3WriteEnabled=true
    Module.Param3WriteEnabled1=true
    Module.Param3WriteEnabled2=true
    Module.Param3WriteEnabled3=true
    Write Parameter Index 1=true
    Write Parameter Index 2=true
    Write Parameter Index 3=true
   )
   &ScaleSpriteSize(Scale Factor=&Vector2DFromFloat(Value=&FloatFromCurve(
      FloatCurve=DataInterface() {propertyValues {Curve {keys [
          {interpMode=RCIM_Cubic, tangentMode=RCTM_User, time=0, value=0}
          <at 0.1 0.35>
          {interpMode=RCIM_Cubic, tangentMode=RCTM_User, time=1, value=1}
         ]}}}
      Scale Curve=3
     )))
   &ScaleSpriteSizeBySpeed(Max Scale Factor=<2 10>, Min Scale Factor=<0.4 2>) {enabled=false}
   &SpriteRotationRate(Rotation Rate=&Multiply_Float(A=750, B=&RandomRangeFloat(Minimum=1, Maximum=-1)))
  ]
  rendererClasses [&NiagaraSpriteRendererProperties]
  renderers [{rendererClass=&NiagaraSpriteRendererProperties}]
 }
 ShieldDecals {
  bEnabled=false
  emitterUpdateScript [
   &EmitterState(Life Cycle Mode=::NewEnumerator1, Loop Behavior=::NewEnumerator1, Loop Delay=0.1, Loop Duration=1, UseLoopDelay=true)
   &SpawnBurst_Instantaneous(
    Spawn Count=20
    Spawn Probability=&MakeCustomFloatFromBool(Bool=$User.bIsFrontEndPreview, True Float=0, False Float=1)
    Spawn Time=0.01
    Use Spawn Probability=true
   )
  ]
  particleSpawnScript [&InitializeParticle(
    Color=<rgba 0.6 2.2751 3 1>
    Lifetime=1.1
    Sprite Size Mode=::NewEnumerator3
    Uniform Sprite Size=50
    UsePositionOffset=false
    Write Material Random=true
   ) {module=InitializeParticle001}]
  particleUpdateScript [
   &ParticleState()
   &Color(Color=&MultiplyLinearColor(
     Linear Color=<rgba 0.1 0.28 1 1>
     Scale Factor=&MakeFloat4FromVectorAndFloat("Vector (RGB)"=&VectorFromFloat(Value=10), "Float (Alpha)"=2)
    ))
   &ScaleColor(
    Color Value To Scale=$Particles.Color
    Scale Alpha=&FloatFromCurve(FloatCurve=DataInterface() {propertyValues {Curve {keys [{arriveTangent=7.9503, interpMode=RCIM_Cubic, time=0, value=0}, <at 0.125 1.25 RCIM_Cubic>, <at 1 0>]}}})
    Scale RGB=&VectorFromFloat(Value=0.5)
    ScaleA=true
    ScaleRGB=true
   )
   &SkeletalMeshLocation(
    Bone / Socket Index=&ReturnParticleIDAttributes()
    Bone Sampling Mode=::NewEnumerator4
    Inherit Velocity=false
    MaskSpawnGroup=false
    Skeletal Mesh=$User.SkeletalMesh
   )
   &CameraOffset(Camera Offset Amount=25)
   &DynamicMaterialParameters(
    Index 0 Param 1=&FloatFromCurve(
     FloatCurve=DataInterface() {propertyValues {Curve {keys [
         <at 0 -0.5 RCIM_Cubic>
         {arriveTangent=10, interpMode=RCIM_Cubic, leaveTangent=10, time=0.0692, value=0.2738}
         {arriveTangent=5.0777, interpMode=RCIM_Cubic, leaveTangent=5.0777, time=0.15, value=1}
         {interpMode=RCIM_Cubic, leaveTangent=4.2839, time=1, value=5}
        ]}}}
     Scale Curve=1.5
    )
    Index 0 Param 3=&FloatFromCurve(
     FloatCurve=DataInterface() {propertyValues {Curve {keys [{arriveTangent=2.3217, interpMode=RCIM_Cubic, time=0, value=-0.25}, {interpMode=RCIM_Cubic, leaveTangent=0.1844, time=1, value=1.4}]}}}
     Scale Curve=2
    )
    Index 0 Param 4=1
    Module.Param0WriteEnabled=true
    Module.Param1WriteEnabled=true
    Module.Param2WriteEnabled=true
    Module.Param3WriteEnabled=true
   )
  ]
  rendererClasses [&NiagaraSpriteRendererProperties]
  renderers [{rendererClass=&NiagaraSpriteRendererProperties}]
 }
 SmokeFingers {
  emitterUpdateScript [
   &EmitterState(Life Cycle Mode=::NewEnumerator1, Loop Behavior=::NewEnumerator1, Loop Delay=0.8083, Loop Duration=1, UseLoopDelay=false)
   &SpawnRate(SpawnRate=&Lerp_Float(A=30, B=0, Alpha=$System.Dive), Use Spawn Probability=false) {enabled=false}
   &SpawnBurst_Instantaneous(Spawn Count=25, Use Spawn Probability=false, UseLoopCountLimit=false)
  ]
  particleSpawnScript [
   &InitializeParticle(
    Color=<rgba 0.6691 0.8446 1 0.15>
    Lifetime=&RandomRangeFloat(Minimum=0.2, Maximum=0.4)
    Position Offset Coordinate Space=::Local
    Sprite Rotation Angle=&RandomRangeFloat(Minimum=0, Maximum=1)
    Sprite Rotation Mode=::NewEnumerator1
    Sprite Size=<5 30>
    Sprite Size Mode=::NewEnumerator4
    UsePositionOffset=true
    Write Material Random=true
   )
   &FN_SetSpawnGroup(Number of Spawn Groups=2) {enabled=false}
   &SkeletalMeshLocation(Bone Sampling Mode=::NewEnumerator4, MaskSpawnGroup=false, Orientation Sampling=::NewEnumerator1, Skeletal Mesh=$User.Skeletal Mesh01) {enabled=false}
   &TorusLocation(
    Align Torus With Axis=false
    Large Radius=5
    MaskSpawnGroup=false
    Module.Override Randomness=false
    Surface Position=1
    Torus Mode=::NewEnumerator2
   )
   &AddVelocityFromPoint(Module.Use Falloff Distance=false, Velocity Strength=&RandomRangeFloat(Minimum=100, Maximum=400))
   &AddVelocity()
   /SetVariables(Particles.Position=&AddVectorToPosition(Position=$Particles.Position, Vector=<0 0 10>)) {isSetParameters=true}
  ]
  particleUpdateScript [
   &Drag(Drag=&RandomRangeFloat(Minimum=8, Maximum=10), Use Linear Drag=true, Use Rotational Drag=false)
   &CurlNoiseForce(Noise Strength=800, Pan Noise=false) {enabled=false}
   &GravityForce(Gravity=<0 0 100>)
   &SolveForcesAndVelocity(Clamp Velocity=false, Limit Acceleration=false)
   &ParticleState()
   &DynamicMaterialParameters(
    Index 0 Param 1=&FloatFromCurve(FloatCurve=DataInterface() {propertyValues {Curve {keys [<at 0 1>,<at 1 2>]}}}, Scale Curve=0.2)
    Index 0 Param 3=&FloatFromCurve(
     FloatCurve=DataInterface() {propertyValues {Curve {keys [
         {arriveTangent=-0.9531, interpMode=RCIM_Cubic, leaveTangent=-0.9531, tangentMode=RCTM_User, time=0, value=0.07}
         {arriveTangent=-0.3362, interpMode=RCIM_Cubic, leaveTangent=-0.3362, tangentMode=RCTM_User, time=0.1, value=0.01}
         {interpMode=RCIM_Cubic, leaveTangent=1.7757, time=0.15, value=0}
        ]}}}
     Scale Curve=0
    )
    Index 1 Param 1=&FloatFromCurve(FloatCurve=DataInterface() {propertyValues {Curve {keys [<at 0 0>,<at 1 1>]}}})
    Index 1 Param 2=&FloatFromCurve(FloatCurve=DataInterface() {propertyValues {Curve {keys [{arriveTangent=1.9694, interpMode=RCIM_Cubic, leaveTangent=1.9694, tangentMode=RCTM_User, time=0, value=0}, <at 1 1 RCIM_Cubic>]}}})
    Index 1 Param 3=&FloatFromCurve(FloatCurve=DataInterface() {propertyValues {Curve {keys [{arriveTangent=6.205, interpMode=RCIM_Cubic, leaveTangent=6.205, tangentMode=RCTM_User, time=0, value=0.2}, <at 0.25 1 RCIM_Cubic>]}}})
    Index 1 Param 4=&FloatFromCurve(FloatCurve=DataInterface() {propertyValues {Curve {keys [<at 0 1>,<at 1 0>]}}})
    Index 2 Param 1=&FloatFromCurve(FloatCurve=DataInterface() {propertyValues {Curve {keys [{arriveTangent=2.2804, interpMode=RCIM_Cubic, leaveTangent=2.2804, tangentMode=RCTM_User, time=0, value=0.1}, <at 1 1 RCIM_Cubic>]}}})
    Index 2 Param 2=&FloatFromCurve(FloatCurve=DataInterface() {propertyValues {Curve {keys [<at 0 0.2>,<at 1 0.8>]}}})
    Index 2 Param 3=&FloatFromCurve(FloatCurve=DataInterface() {propertyValues {Curve {keys [{interpMode=RCIM_Cubic, tangentMode=RCTM_User, time=0, value=1}, {interpMode=RCIM_Cubic, tangentMode=RCTM_User, time=1, value=2}]}}})
    Index 2 Param 4=&FloatFromCurve(FloatCurve=DataInterface() {propertyValues {Curve {keys [<at 0 1 RCIM_Cubic>,<at 1 0.75 RCIM_Cubic>]}}})
    Module.Param0WriteEnabled=true
    Module.Param0WriteEnabled1=true
    Module.Param0WriteEnabled2=true
    Module.Param0WriteEnabled3=true
    Module.Param1WriteEnabled=true
    Module.Param1WriteEnabled1=true
    Module.Param1WriteEnabled2=true
    Module.Param1WriteEnabled3=true
    Module.Param2WriteEnabled=true
    Module.Param2WriteEnabled1=true
    Module.Param2WriteEnabled2=true
    Module.Param2WriteEnabled3=true
    Module.Param3WriteEnabled=true
    Module.Param3WriteEnabled1=true
    Module.Param3WriteEnabled2=true
    Module.Param3WriteEnabled3=true
    Write Parameter Index 1=true
    Write Parameter Index 2=true
    Write Parameter Index 3=true
   )
   &ScaleSpriteSize(Scale Factor=&Vector2DFromFloat(Value=&FloatFromCurve(
      FloatCurve=DataInterface() {propertyValues {Curve {keys [{interpMode=RCIM_Cubic, tangentMode=RCTM_User, time=0, value=0.5}, {interpMode=RCIM_Cubic, tangentMode=RCTM_User, time=1, value=1}]}}}
      Scale Curve=4
     )))
   &ScaleSpriteSize(
    Non-Uniform Curve Sprite Scale=DataInterface() {propertyValues {
      XCurve {keys [{tangentMode=RCTM_User, time=0, value=0.5}, {tangentMode=RCTM_User, time=1, value=1}]}
      YCurve {keys [
        {arriveTangent=0.5946, interpMode=RCIM_Cubic, leaveTangent=0.5946, tangentMode=RCTM_User, time=0, value=0}
        {arriveTangent=1.6169, interpMode=RCIM_Cubic, leaveTangent=1.6169, tangentMode=RCTM_User, time=1, value=1}
       ]}
     }}
    Scale Sprite Size Mode=::NewEnumerator3
   ) {module=ScaleSpriteSize001}
   &ScaleColor(
    Scale Alpha=&FloatFromCurve(FloatCurve=DataInterface() {propertyValues {Curve {keys [
         {arriveTangent=9.0725, interpMode=RCIM_Cubic, leaveTangent=9.0725, tangentMode=RCTM_User, time=0, value=0}
         {
          arriveTangent=-2.4885
          arriveTangentWeight=0.1274
          interpMode=RCIM_Cubic
          leaveTangent=-2.4885
          leaveTangentWeight=0.7778
          tangentMode=RCTM_User
          time=0.1408
          value=1
         }
         {
          arriveTangentWeight=0.6167
          interpMode=RCIM_Cubic
          leaveTangentWeight=0.6167
          tangentMode=RCTM_User
          tangentWeightMode=RCTWM_WeightedBoth
          time=1
          value=0
         }
        ]}}})
    ScaleA=true
    ScaleRGB=true
   )
  ]
  rendererClasses [&NiagaraSpriteRendererProperties]
  renderers [{rendererClass=&NiagaraSpriteRendererProperties}]
 }
 SmokeGroundRings {
  emitterUpdateScript [
   &EmitterState()
   &SpawnBurst_Instantaneous(Spawn Count=2, Use Spawn Probability=false, UseLoopCountLimit=false)
   &SpawnRate(SpawnRate=5, Use Spawn Probability=false) {enabled=false}
  ]
  particleSpawnScript [
   &PartitionParticles(
    Normalized Index In Partition Includes One=true
    Number Of Partitions=1
    Particles Per Partition=2
    Partition Mode:ENiagara_PartitionMode=::NewEnumerator1
   )
   &InitializeParticle(
    Color=<rgba 1 1 1 0.2>
    Lifetime=0.25
    Sprite Rotation Mode=::NewEnumerator2
    Sprite Size Mode=::NewEnumerator3
    Uniform Sprite Size=&Lerp_Float(A=2.5, B=1.875, Alpha=$Particles.PartitionParticles.Normalized Index In Partition)
    UsePositionOffset=false
    Write Material Random=true
   ) {module=InitializeParticle001}
   /SetVariables(Particles.SpriteAlignment=&RandomVector(Vector Scale=1), Particles.SpriteFacing=<0 0 1>) {isSetParameters=true}
  ]
  particleUpdateScript [
   &ParticleState()
   &DynamicMaterialParameters(
    Index 0 Param 1=&FloatFromCurve(FloatCurve=DataInterface() {propertyValues {Curve {keys [<at 0 0.5>,<at 1 2>]}}})
    Module.Param0WriteEnabled=true
    Module.Param1WriteEnabled=false
    Module.Param2WriteEnabled=false
    Module.Param3WriteEnabled=false
   )
   &ScaleSpriteSize(Scale Factor=&Vector2DFromFloat(Value=&FloatFromCurve(
      FloatCurve=DataInterface() {propertyValues {Curve {keys [
          {arriveTangent=35.6763, interpMode=RCIM_Cubic, leaveTangent=3.7835, tangentMode=RCTM_User, time=0, value=0}
          {interpMode=RCIM_Cubic, leaveTangent=0.6903, time=1.0011, value=1}
         ]}}}
      Scale Curve=50
     )))
   &CameraOffset(Camera Offset Amount=20)
   &ScaleColor(
    Scale Alpha=&FloatFromCurve(FloatCurve=DataInterface() {propertyValues {Curve {keys [<at 0 1 RCIM_Cubic>, {interpMode=RCIM_Cubic, leaveTangent=-4.9981, time=1.0011, value=0.002}]}}})
    ScaleA=true
    ScaleRGB=false
   )
  ]
  rendererClasses [&NiagaraSpriteRendererProperties]
  renderers [{rendererClass=&NiagaraSpriteRendererProperties}]
 }
 SmokeLoop {
  emitterUpdateScript [
   &EmitterState()
   &SpawnBurst_Instantaneous(
    Spawn Count=&RandomRangeInt(Minimum=40, Maximum=30)
    Spawn Probability=&MakeCustomFloatFromBool(Bool=$User.bIsFrontEndPreview, True Float=0, False Float=1)
    Use Spawn Probability=true
    UseLoopCountLimit=false
   )
  ]
  particleSpawnScript [
   /SetVariables(Particles.SpriteRotationRate=&RandomRangeFloat(Minimum=-0.8, Maximum=0.8, Evaluation Type=::NewEnumerator1)) {isSetParameters=true, module=SetVariables_#1}
   &InitializeParticle(
    Color=&FNTimeOfDayColor(Night Color=<rgba 1 1 1 0.035>, Morning Color=<rgba 1 1 1 0.035>, Day Color=<rgba 1 1 1 0.035>, Evening Color=<rgba 1 1 1 0.035>)
    Lifetime Max=0.8
    Lifetime Min=0.6
    Lifetime Mode=::NewEnumerator1
    Sprite Rotation Mode=::NewEnumerator2
    Sprite Size Mode=::NewEnumerator1
    Uniform Sprite Size Min=15
    UsePositionOffset=false
    Write Material Random=true
   ) {module=InitializeParticle001}
   &CameraOffset(Camera Offset Amount=5)
   &AddVelocity(
    Cone Angle=160
    Cone Axis=<0 0 1>
    Inner Cone Angle=120
    Module.UseConeFalloff=false
    Module.UseVelDistribution=false
    Override Seed=false
    Rotation Coordinate Space=::World
    Velocity Mode=::NewEnumerator2
    Velocity Speed=175
   )
  ]
  particleUpdateScript [
   &ParticleState()
   &SpriteRotationRate(Rotation Rate=&RandomRangeFloat(Minimum=-0.8, Maximum=0.8))
   &ScaleColor(
    Scale Alpha=&FloatFromCurve(
     FloatCurve=DataInterface() {propertyValues {Curve {keys [
         {arriveTangent=9.0725, interpMode=RCIM_Cubic, time=0, value=0}
         {arriveTangent=-2.5249, interpMode=RCIM_Cubic, leaveTangent=-2.5249, tangentMode=RCTM_User, time=0.1408, value=0.9847}
         {arriveTangent=-0.0405, interpMode=RCIM_Cubic, leaveTangent=-0.0405, tangentMode=RCTM_User, time=1, value=0}
        ]}}}
     Scale Curve=0.35
    )
    ScaleA=true
    ScaleRGB=true
   )
   &ScaleSpriteSize(Scale Factor=&Vector2DFromFloat(Value=&FloatFromCurve(
      FloatCurve=DataInterface() {propertyValues {Curve {keys [
          {arriveTangent=1.9916, interpMode=RCIM_Cubic, time=0, value=0.5}
          {arriveTangent=0.8423, interpMode=RCIM_Cubic, leaveTangent=0.8423, tangentMode=RCTM_User, time=0.2, value=0.75}
          <at 1 1 RCIM_Cubic>
         ]}}}
      Scale Curve=2
     )))
   &AccelerationForce(Acceleration=<0 0 40>)
   &Drag(
    Drag=&FloatFromCurve(FloatCurve=DataInterface() {propertyValues {Curve {keys [{interpMode=RCIM_Cubic, tangentMode=RCTM_User, time=0, value=5}, {interpMode=RCIM_Cubic, tangentMode=RCTM_User, time=1, value=6}]}}})
    Use Linear Drag=true
    Use Rotational Drag=false
   )
   &CurlNoiseForce(Noise Quality / Cost=::NewEnumerator0, Noise Strength=1000, Pan Noise=true, Pan Noise Field=<0 0 1>) {enabled=false}
   &SolveForcesAndVelocity(Clamp Velocity=false, Limit Acceleration=false)
   &DynamicMaterialParameters(
    Index 0 Param 1=&FloatFromCurve(FloatCurve=DataInterface() {propertyValues {Curve {keys [<at 0 1>,<at 1 0>]}}}, Scale Curve=0.2)
    Module.Param0WriteEnabled=true
    Module.Param1WriteEnabled=false
    Module.Param2WriteEnabled=false
    Module.Param3WriteEnabled=false
   )
  ]
  rendererClasses [&NiagaraSpriteRendererProperties]
  renderers [{rendererClass=&NiagaraSpriteRendererProperties}]
 }
 SmokeLoop001 {
  bEnabled=false
  emitterUpdateScript [
   &EmitterState()
   &SpawnBurst_Instantaneous(
    Spawn Count=&RandomRangeInt(Minimum=30, Maximum=40)
    Spawn Probability=&MakeCustomFloatFromBool(Bool=$User.bIsFrontEndPreview, True Float=0, False Float=1)
    Use Spawn Probability=true
    UseLoopCountLimit=false
   )
  ]
  particleSpawnScript [
   &FN_SetSpawnGroup(Number of Spawn Groups=16)
   /SetVariables(Particles.SpriteRotationRate=&RandomRangeFloat(Minimum=-0.8, Maximum=0.8, Evaluation Type=::NewEnumerator1)) {isSetParameters=true, module=SetVariables_#1}
   &InitializeParticle(
    Color=&FNTimeOfDayColor(
     Night Color=<rgba 0.5 0.7241 1 1>
     Morning Color=<rgba 0.5 0.7241 1 1>
     Day Color=<rgba 0.42 0.6799 1 1>
     Evening Color=<rgba 0.5 0.7241 1 1>
    )
    Lifetime Max=0.8
    Lifetime Min=0.6
    Lifetime Mode=::NewEnumerator1
    Sprite Rotation Mode=::NewEnumerator2
    Sprite Size Mode=::NewEnumerator1
    Uniform Sprite Size Max=100
    Uniform Sprite Size Min=80
    UsePositionOffset=false
    Write Material Random=true
   ) {module=InitializeParticle001}
   &CameraOffset(Camera Offset Amount=&RandomRangeFloat(Minimum=35, Maximum=45))
   &ShapeLocation(
    Cylinder Radius=10
    MaskSpawnGroup=false
    Offset=<0 0 87.5>
    Offset Mode=::NewEnumerator0
    Override Seed=false
    Shape Primitive=::NewEnumerator1
    Surface Only=true
    Surface Only Band Thickness:NiagaraFloat=&RandomRangeFloat(Minimum=0, Maximum=0.25)
    Use Endcaps In Surface Only Mode=false
    UseOwnerScale=true
   )
   &AddVelocity(Velocity=&Multiply_VectorByFloat(Vector=$Particles.ShapeLocation.ShapeVector, Float=3)) {enabled=false}
   &AddVelocity(Velocity=&RandomRangeVector(Minimum=<0 0 15>, Maximum=<0 0 25>, Evaluation Type=::NewEnumerator1)) {enabled=false, module=AddVelocity001}
   &AddVelocity(
    Velocity=$Particles.ShapeLocation.ShapeNormal
    Velocity Speed Scale=&RandomRangeFloat(Minimum=0, Maximum=500)
   ) {module=AddVelocity003}
   &AddVelocity(Module.Use Falloff Distance=false, Origin Offset=<0 0 87.5>, Override Seed=false, Velocity Mode=::NewEnumerator1, Velocity Speed=200) {module=AddVelocity002}
  ]
  particleUpdateScript [
   &ParticleState()
   &SpriteRotationRate(Rotation Rate=&RandomRangeFloat(Minimum=-0.8, Maximum=0.8))
   &ScaleColor(
    Scale Alpha=&FloatFromCurve(
     FloatCurve=DataInterface() {propertyValues {Curve {keys [
         {arriveTangent=9.0725, interpMode=RCIM_Cubic, time=0, value=0}
         {arriveTangent=-2.5249, interpMode=RCIM_Cubic, leaveTangent=-2.5249, tangentMode=RCTM_User, time=0.1408, value=0.9847}
         {arriveTangent=-0.0405, interpMode=RCIM_Cubic, leaveTangent=-0.0405, tangentMode=RCTM_User, time=1, value=0}
        ]}}}
     Scale Curve=0.35
    )
    ScaleA=true
    ScaleRGB=true
   )
   &ScaleSpriteSize(Scale Factor=&Vector2DFromFloat(Value=&FloatFromCurve(FloatCurve=DataInterface() {propertyValues {Curve {keys [
          {arriveTangent=1.9916, interpMode=RCIM_Cubic, time=0, value=0.5}
          {arriveTangent=0.5, interpMode=RCIM_Cubic, leaveTangent=0.5, time=0.2, value=0.75}
          <at 1 1 RCIM_Cubic>
         ]}}})))
   &AccelerationForce(Acceleration=<0 0 40>)
   &Drag(
    Drag=&FloatFromCurve(FloatCurve=DataInterface() {propertyValues {Curve {keys [{interpMode=RCIM_Cubic, tangentMode=RCTM_User, time=0, value=5}, {interpMode=RCIM_Cubic, tangentMode=RCTM_User, time=1, value=8}]}}})
    Use Linear Drag=true
    Use Rotational Drag=false
   )
   &CurlNoiseForce(Noise Quality / Cost=::NewEnumerator0, Noise Strength=1000, Pan Noise=true, Pan Noise Field=<0 0 1>)
   &SolveForcesAndVelocity(Clamp Velocity=false, Limit Acceleration=false)
   &DynamicMaterialParameters(
    Index 0 Param 1=&FloatFromCurve(FloatCurve=DataInterface() {propertyValues {Curve {keys [<at 0 1>,<at 1 0>]}}}, Scale Curve=0.2)
    Module.Param0WriteEnabled=true
    Module.Param1WriteEnabled=false
    Module.Param2WriteEnabled=false
    Module.Param3WriteEnabled=false
   )
  ]
  rendererClasses [&NiagaraSpriteRendererProperties]
  renderers [{rendererClass=&NiagaraSpriteRendererProperties}]
 }
 SpinSplines {
  bEnabled=false
  emitterUpdateScript [
   &EmitterState()
   &SpawnBurst_Instantaneous(
    Spawn Count=20
    Spawn Probability=&MakeCustomFloatFromBool(Bool=$User.bIsFrontEndPreview, True Float=0, False Float=1)
    Use Spawn Probability=true
    UseLoopCountLimit=false
   ) {module=SpawnBurst_Instantaneous001}
  ]
  particleSpawnScript [
   &InitializeParticle(
    Color=<rgba 0.264 0.5278 0.8 0.85>
    Lifetime Max=0.9
    Lifetime Min=0.65
    Lifetime Mode=::NewEnumerator1
    Mesh Scale=&Multiply_VectorByFloat(Vector=&RandomRangeVector(Minimum=<0.1 0.15 0.15>, Maximum=<0.4 0.2 0.2>), Float=2)
    Mesh Scale Mode=::NewEnumerator4
    UsePositionOffset=false
    Write Material Random=true
   ) {module=InitializeParticle001}
   &SkeletalMeshLocation(Bone Sampling Mode=::NewEnumerator0, MaskSpawnGroup=false, OverrideRandomSeed=false, Skeletal Mesh=$User.SkeletalMesh)
   &InitialMeshOrientation(
    Orientation Axis=<0 0 -1>
    Rotation=&MakeVector(
     X=&RandomRangeFloat(Minimum=0, Maximum=1)
     Y=&FloatFromCurve(
      FloatCurve=DataInterface() {propertyValues {Curve {keys [
          <at 0 1 RCIM_Cubic>
          {arriveTangent=-5.2683, interpMode=RCIM_Cubic, leaveTangent=-5.2682, tangentMode=RCTM_User, time=0.4932, value=0}
          <at 1 -1>
         ]}}}
      CurveIndex=&RandomRangeFloat(Minimum=0, Maximum=1)
      Scale Curve=0.08
     )
     Z=0
    )
    Use Orientation Vector=true
    Use Rotation Vector=true
   ) {module=InitialMeshOrientation001}
  ]
  particleUpdateScript [
   &ParticleState()
   &ScaleColor(
    Scale Alpha=&FloatFromCurve(
     FloatCurve=DataInterface() {propertyValues {Curve {keys [
         {arriveTangent=0.0393, interpMode=RCIM_Cubic, leaveTangent=0.0393, tangentMode=RCTM_User, time=0.5, value=1.5}
         {arriveTangent=-0.0943, interpMode=RCIM_Cubic, leaveTangent=-0.0943, tangentMode=RCTM_User, time=1, value=0}
        ]}}}
     Scale Curve=0.8
    )
    Scale RGB=&VectorFromFloat(Value=&FloatFromCurve(FloatCurve=DataInterface() {propertyValues {Curve {keys [<at 0 0 RCIM_Cubic>,<at 0.1 1 RCIM_Cubic>,<at 0.75 0 RCIM_Cubic>]}}}))
    ScaleA=true
    ScaleRGB=false
   )
   &DynamicMaterialParameters(
    Index 0 Param 1=&FloatFromCurve(FloatCurve=DataInterface() {propertyValues {Curve {keys [
         {interpMode=RCIM_Cubic, tangentMode=RCTM_User, time=0, value=1.25}
         {arriveTangent=-0.75, interpMode=RCIM_Cubic, leaveTangent=-0.75, time=0.1, value=1}
         {interpMode=RCIM_Cubic, tangentMode=RCTM_User, time=1, value=0.5}
        ]}}})
    Index 0 Param 2=&FloatFromCurve(
     FloatCurve=DataInterface() {propertyValues {Curve {keys [
         {tangentMode=RCTM_User, time=0, value=1}
         {arriveTangent=-0.0405, interpMode=RCIM_Cubic, leaveTangent=-0.0405, tangentMode=RCTM_User, time=0.7009, value=1}
         {arriveTangent=-8.0089, interpMode=RCIM_Cubic, leaveTangent=-8.0089, tangentMode=RCTM_User, time=1, value=0}
        ]}}}
     Scale Curve=2
    )
    Index 0 Param 4=&RandomRangeFloat(Minimum=0, Maximum=1)
    Module.Param0WriteEnabled=true
    Module.Param1WriteEnabled=true
    Module.Param2WriteEnabled=false
    Module.Param3WriteEnabled=true
   )
   &UpdateMeshOrientation(
    Rotation Coordinate Space=::NewEnumerator3
    Rotation Rate=&FloatFromCurve(
     FloatCurve=DataInterface() {propertyValues {Curve {keys [
         {arriveTangent=7.0045, interpMode=RCIM_Cubic, leaveTangent=7.0045, tangentMode=RCTM_User, time=0, value=0.2}
         <at 0.25 1 RCIM_Cubic>
         <at 1 0.2 RCIM_Cubic>
        ]}}}
     Scale Curve=&RandomRangeFloat(Minimum=-0.3, Maximum=-0.16)
    )
    Rotation Vector=<1 0 0>
   ) {module=UpdateMeshOrientation001}
   &ScaleMeshSize(Scale Factor=&VectorFromFloat(Value=&FloatFromCurve(FloatCurve=DataInterface() {propertyValues {Curve {keys [
          {arriveTangent=0.0602, interpMode=RCIM_Cubic, leaveTangent=0.0602, tangentMode=RCTM_User, time=0, value=0}
          {arriveTangent=0.0697, interpMode=RCIM_Cubic, leaveTangent=0.0697, tangentMode=RCTM_User, time=0.8, value=1}
         ]}}})))
   &ScaleMeshSize(Scale Factor=&MakeVector(
     X=1
     Y=&FloatFromCurve(FloatCurve=DataInterface() {propertyValues {Curve {keys [
          {arriveTangent=2.4016, interpMode=RCIM_Cubic, leaveTangent=2.4016, tangentMode=RCTM_User, time=0, value=0.5}
          {interpMode=RCIM_Cubic, tangentMode=RCTM_User, time=0.6, value=1}
         ]}}})
     Z=&FloatFromCurve(FloatCurve=DataInterface() {propertyValues {Curve {keys [
          {arriveTangent=2.4016, interpMode=RCIM_Cubic, leaveTangent=2.4016, tangentMode=RCTM_User, time=0, value=0.5}
          {interpMode=RCIM_Cubic, tangentMode=RCTM_User, time=0.6, value=1}
         ]}}})
    )) {module=ScaleMeshSize001}
   &SolveForcesAndVelocity(Clamp Velocity=false, Limit Acceleration=false)
  ]
  rendererClasses [&NiagaraMeshRendererProperties]
  renderers [{rendererClass=&NiagaraMeshRendererProperties, rendererIndex=0}]
 }
 Streaks001 {
  emitterUpdateScript [
   &EmitterState(Life Cycle Mode=::NewEnumerator1, Loop Behavior=::NewEnumerator1, Loop Duration=0.15, UseLoopDelay=false)
   &SpawnRate(SpawnRate=300, Use Spawn Probability=false)
  ]
  particleSpawnScript [
   /Niagara/Modules/Spawn/Initialization/InitializeParticle(
    Color=&MakeLinearColorFromVectorAndFloat("Vector (RGB)"=&MakeVectorFromLinearColorRGB(Color=<rgba 0.629 0.7403 1 1>), "Float (Alpha)"=0.1)
    Lifetime=&UniformRangedFloat(Minimum=0.3, Maximum=0.5)
    Module.Write Color=true
    Module.Write Lifetime=true
    Module.Write Mass=true
    Module.Write Material Random=true
    Module.Write Position=false
    Module.Write Scale=false
    Module.Write SpriteRotation=true
    Module.Write SpriteSize=true
    Position=&SimulationPosition()
    Sprite Size=&Multiply_Vector2(
     A=&UniformRangedVector2D(Minimum=<20 116.296>, Maximum=<40 209.1042>)
     B=&Vector2DFromFloat(Value=0.3)
    )
   ) {enabled=false}
   /Niagara/Modules/Spawn/Initialization/V2/InitializeParticle(
    Color=<rgba 0.629 0.7403 1 0.2>
    Lifetime Max=0.05
    Lifetime Min=0.15
    Lifetime Mode=::NewEnumerator1
    Sprite Size Max=<10 18>
    Sprite Size Min=<5 12>
    Sprite Size Mode=::NewEnumerator2
    UsePositionOffset=false
    Write Material Random=true
   ) {module=InitializeParticle001}
   &AddVelocity(
    Cone Angle=160
    Cone Axis=<0 0 -1>
    Inner Cone Angle=120
    Module.UseConeFalloff=false
    Module.UseVelDistribution=false
    Override Seed=false
    Velocity Mode=::NewEnumerator2
    Velocity Speed=300
   ) {module=AddVelocity001}
   &AddVelocity(Velocity=&MakeVector(X=0, Y=0, Z=&UniformRangedFloat(Minimum=50, Maximum=100)))
   &ApplyInitialForces()
   &DynamicMaterialParameters(
    Index 0 Param 1=&UniformRangedFloat(Minimum=0, Maximum=1)
    Index 0 Param 2=&UniformRangedFloat(Minimum=1, Maximum=2)
    Index 0 Param 3=1
    Index 0 Param 4=1
    Module.Param0WriteEnabled=true
    Module.Param1WriteEnabled=true
    Module.Param2WriteEnabled=true
    Module.Param3WriteEnabled=true
   )
   /SetVariables(Particles.Position=&AddVectorToPosition(Position=$Particles.Position, Vector=<0 0 10>)) {isSetParameters=true}
  ]
  particleUpdateScript [
   &ParticleState()
   &Drag(Drag=0, Use Linear Drag=true, Use Rotational Drag=false)
   &SolveForcesAndVelocity(Clamp Velocity=false, Limit Acceleration=false)
   &ScaleColor(
    Scale Alpha=&FloatFromCurve(
     FloatCurve=DataInterface() {propertyValues {Curve {keys [<at 0 0 RCIM_Cubic>, {arriveTangent=0.002, interpMode=RCIM_Cubic, leaveTangent=0.002, time=0.5, value=1}, <at 1.0011 0.002 RCIM_Cubic>]}}}
     Scale Curve=0.8
    )
    Scale RGB=&MakeVectorFromLinearColorRGB(Color=&ColorFromCurve(ColorCurve=DataInterface() {propertyValues {
        AlphaCurve {keys [<at 0 1 RCIM_Cubic>,<at 1 0 RCIM_Cubic>]}
        BlueCurve {keys [<at 0 1.14 RCIM_Cubic>,<at 1 1 RCIM_Cubic>]}
        GreenCurve {keys [<at 0 1.842 RCIM_Cubic>,<at 1 1 RCIM_Cubic>]}
        RedCurve {keys [<at 0 3 RCIM_Cubic>,<at 1 1 RCIM_Cubic>]}
       }}))
    ScaleA=true
    ScaleRGB=true
   )
  ]
  rendererClasses [&NiagaraSpriteRendererProperties]
  renderers [{rendererClass=&NiagaraSpriteRendererProperties}]
 }
}
summary {
 systemName=NS_Emote_Flourish_Swap
 emitters [
  {bEnabled=false, emitterName=BurstEmbers, rendererClasses [&NiagaraSpriteRendererProperties]}
  {bEnabled=false, emitterName=ShieldDecals, rendererClasses [&NiagaraSpriteRendererProperties]}
  {bEnabled=false, emitterName=SpinSplines, rendererClasses [&NiagaraMeshRendererProperties]}
  {bEnabled=false, emitterName=CircularSmokeRate, rendererClasses [&NiagaraSpriteRendererProperties]}
  {bEnabled=false, emitterName=SmokeLoop001, rendererClasses [&NiagaraSpriteRendererProperties]}
  {emitterName=SmokeFingers, rendererClasses [&NiagaraSpriteRendererProperties]}
  {emitterName=SmokeGroundRings, rendererClasses [&NiagaraSpriteRendererProperties]}
  {emitterName=Streaks001, rendererClasses [&NiagaraSpriteRendererProperties]}
  {emitterName=SmokeLoop, rendererClasses [&NiagaraSpriteRendererProperties]}
 ]
 userVariables [
  User.bIsFrontEnd {type=NiagaraBool}
  User.bIsFrontEndPreview {type=NiagaraBool}
  User.bIsInWater {type=NiagaraBool}
  User.SkeletalMesh {
   type=NiagaraDataInterfaceSkeletalMesh
   defaultValue {
    struct=NiagaraExt_VariableValue_DataInterface
    value {dataInterface=&self:NiagaraDataInterfaceSkeletalMesh_0}
   }
  }
 ]
}
